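# CreativeLevelListItem is a CommonButtonBase list row that shows one selectable level option in the Creative property editor's level list.
# On list item assignment it caches the wrapper's LevelOption, sets the display name and toggles a new-level bang.
# Selection takes focus, clicking broadcasts OnLevelCommitted, and button style changes restyle the text.
# IsRepresentingLevelObject matches a wrapper to the row by level instance GUID.
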WidgetBlueprint /FortniteCoreUI/InGame/Creative/PropertyEditor/Level_List/CreativeLevelListItem
graphs [
 EventGraph {nodes [
   Event OnCurrentTextStyleChanged() {is=title K2Node_Event}
   > UpdateTextStyle()
   n1 = Event OnListItemObjectSet()
   > n2 = Cast To LevelOptionObjectWrapper!(execute, Object=@n1.ListItemObject) -> (then, CastFailed, AsLevel Option Object Wrapper:Level Option Object Wrapper Object Reference) {is=title K2Node_DynamicCast}
   > n3 = Set AvailableLevelOption(AvailableLevelOption:Playlist Option Level Value Structure=Get LevelOption!(self:Level Option Object Wrapper Object Reference=@n2) -> (LevelOption:Playlist Option Level Value Structure)) -> (Output_Get:Playlist Option Level Value Structure)
   n4 = Break Playlist Option Level Value!(PlaylistOptionLevelValue:"Playlist Option Level Value Structure (by ref)"=@n3) -> (
    LevelInstanceGuid:Guid Structure
    LevelName:String
    bNewLevel:bool=~
    DisplayName:Text
    OptionValueName:String
    GrantedContextTags:Gameplay Tag Container Structure
   ) {is=title K2Node_BreakStruct}
   > Set Bang State!(
    execute
    self:Fort Bang Wrapper NUI Object Reference=Get BangWrapper_NewLevel() -> (BangWrapper_NewLevel:Normal Bang Wrapper Object Reference)
    bEnabled:bool=@n4.bNewLevel
    Count:int=0
   ) -> (then) {is=title K2Node_CallFunction}
   > "SetText (Text)"(self=Get CenterButtonTextWidget() -> (CenterButtonTextWidget:Common Text Object Reference), InText=@n4.DisplayName)
   Event On Selected()
   > SetFocus()
   Event On Clicked()
   > Call On Level Committed!(execute:Exec, self:Creative Level List Item Object Reference, bUseListname:bool=true) -> (then:Exec) {is=title K2Node_CallDelegate}
  ]}
 IsRepresentingLevelObject {nodes [
   n1 = IsRepresentingLevelObject!() -> (then, InLevelObject:Level Option Object Wrapper Object Reference) {is=title K2Node_FunctionEntry}
   > n2 = Is Valid(InputObject=@n1)
   Return Node!(execute=@n2.Is Not Valid, IsObject:bool=false)
   n3 = Get LevelOption(self=@n1)
   n4 = Get AvailableLevelOption() -> (AvailableLevelOption:Playlist Option Level Value Structure)
   > Return Node(IsObject="Equal (Guid)"(
     A:"Guid Structure (by ref)"=Break Playlist Option Level Value(PlaylistOptionLevelValue=@n3) -> (bNewLevel=~)
     B:"Guid Structure (by ref)"=Break Playlist Option Level Value(PlaylistOptionLevelValue=@n4) -> (bNewLevel=~)
    ) -> (ReturnValue:bool=~) {is=title K2Node_PromotableOperator})
  ]}
 OnLevelCommitted {nodes [OnLevelCommitted!() -> (then, bUseListname:bool=~) {is=title K2Node_FunctionEntry}]}
 UpdateTextStyle {nodes [
   UpdateTextStyle() {is=title K2Node_FunctionEntry}
   > SetStyle!(
    execute
    self:Common Text Object Reference=Get CenterButtonTextWidget()
    InStyle:Common Text Style Class Reference=GetCurrentTextStyleClass!(self:Common Button Base Object Reference) -> (ReturnValue:Common Text Style Class Reference)
   ) -> (then)
  ]}
]
variables [
 AvailableLevelOption {
  type=PlaylistOptionLevelValue
  default {
   bNewLevel=false
   displayName=""
   levelInstanceGuid=#1
   levelName=""
   optionValueName=""
   grantedContextTags {gameplayTags []}
   optionalDisplayIcon {largeIcon=null, smallIcon=null}
  }
 }
 OnLevelCommitted {}
]
widgets {
 info {inheritedWidgetCount=1, namedSlotCount=1, parentClass=&CommonButtonBase, rootWidgetClass=&Overlay, widgetCount=5}
 widgets [
  Overlay_35 {widgetClassPath=&Overlay, widget=Overlay_35}
  BangWrapper_NewLevel {
   widgetClassPath=/Game/UI/Foundation/Bangs/NormalBangWrapper
   bIsVariable=true
   parent=Overlay_35
   widget=BangWrapper_NewLevel
   properties {bangType=Custom, useSmallVersion=true}
   slot {class=OverlaySlot}
  }
  SizeBox_0 {
   widgetClassPath=&SizeBox
   namedSlotHost=BangWrapper_NewLevel
   widget=SizeBox_0
   properties {bOverride_WidthOverride=true, widthOverride=350}
  }
  CenterButtonTextWidget {
   widgetClassPath=&CommonTextBlock
   bIsVariable=true
   parent=SizeBox_0
   widget=CenterButtonTextWidget
   properties {
    scrollStyle=&TextScrollStyle-Base_C
    style=/FortUILibrary/Foundation/Buttons/TextStyles/TextStyle-Button-Primary-M_Skew_BurbankLightBlue.TextStyle-Button-Primary-M_Skew_BurbankLightBlue_C
    text=TextBlock_LevelName
    colorAndOpacity {colorUseRule=UseColor_Specified, specifiedColor=<rgba 0.2159 0.6584 1 1>}
    font {
     bForceMonospaced=false
     bMaterialIsStencil=false
     fontMaterial=null
     fontObject=&BurbankSmall
     letterSpacing=0
     monospacedWidth=1
     size=20
     skewAmount=0
     typefaceFontName=Black
     outlineSettings {
      bApplyOutlineToDropShadows=false
      bMiteredCorners=false
      bSeparateFillAlpha=false
      outlineMaterial=null
      outlineSize=0
      outlineColor=<rgba 0 0 0 1>
     }
    }
    shadowOffset=<0 0>
   }
   slot {class=SizeBoxSlot}
  }
  InputActionWidget {widgetClassPath=&CommonActionWidget, bInherited=true, widget=null}
 ]
}
