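# Creative property editor dialog, derived from FortCreativePropertyEditorLevelListDialog, for picking a level from a list.
# Selecting an entry copies its name into a header field, read-only unless rename is enabled.
# Checkmark or Accept broadcasts the index and name via OnLevelSelectionCommited, or OnLevelListClosed if nothing valid is selected.
# As list entries generate and release, it tracks the selected entry's widget for commit binding and focus.
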WidgetBlueprint /FortniteCoreUI/InGame/Creative/PropertyEditor/Level_List/CreativePropertyEditorLevelList
animations [Anim_IntroOutro {
  binding_count=1
  length_seconds=0.25
  object=&self:Anim_IntroOutro
  sequence {bindings [Overlay_18 {
     id=#1
     possessed_class=Overlay
     tracks [
      {
       type=MovieSceneFloatTrack
       display_name=Render Opacity
       sections [{type=MovieSceneFloatSection, end_frame=5, start_frame=0, channels [None {type=MovieSceneScriptingFloatChannel, keys [[0,0],[3,0],[5,1]]}]}]
      }
      {
       type=MovieScene2DTransformTrack
       sections [{
         type=MovieScene2DTransformSection
         end_frame=5
         start_frame=0
         channels [Translation.X {type=MovieSceneScriptingFloatChannel, keys [[0,-440],[5,0]]}]
        }]
      }
     ]
    }]}
 }]
graphs [
 BP_GetDesiredFocusTarget {nodes [
   Get Desired Focus Target()
   > Return Node!(execute, ReturnValue:Widget Object Reference=Get CurrentLevelButton() -> (CurrentLevelButton:Creative Level List Item Object Reference))
  ]}
 CommitNameChange {nodes [
   CommitNameChange() {is=title K2Node_FunctionEntry}
   Get FilteredTextEntry_Widget() -> (FilteredTextEntry_Widget:Fort Editable Filtered Counted Text Box Object Reference)
  ]}
 CommitSelectedLevel {nodes [
   n1 = CommitSelectedLevel!() -> (then, bUseListName:bool=~) {is=title K2Node_FunctionEntry}
   > n2 = GetCurrentText!(
    execute
    self:WBP UIKit Filtered Input Field Object Reference=Get As Filtered Text Entry Widget() -> (As Filtered Text Entry Widget:WBP UIKit Filtered Input Field Object Reference)
   ) -> (then, Text:Text)
   > n3 = Cast To LevelOptionObjectWrapper!(
    execute
    Object=Get Selected Item!(self:List View Object Reference=Get ListView_Levels() -> (ListView_Levels:Common List View Object Reference)) -> (ReturnValue:Object Reference) {is=title K2Node_CallFunction}
   ) -> (then, CastFailed, AsLevel Option Object Wrapper:Level Option Object Wrapper Object Reference) {is=title K2Node_DynamicCast}
   > Call On Level Selection Commited!(
    execute:Exec
    self:Creative Property Editor Level List Object Reference
    Index:int=GetIndexForItem!(self:List View Object Reference=Get ListView_Levels(), Item:Object Reference=@n3) -> (ReturnValue:int=0)
    DisplayText:Text=Select(
     Option 0:Text=@n2
     Option 1:Text=Get!(self:Level Option Object Wrapper Object Reference=@n3) -> (
      LevelOption_LevelInstanceGuid:Guid Structure
      LevelOption_LevelName:String
      LevelOption_bNewLevel:bool=~
      LevelOption_DisplayName:Text
      LevelOption_OptionValueName:String
      LevelOption_GrantedContextTags:Gameplay Tag Container Structure
     ).LevelOption_DisplayName
     Index=@n1
    ) -> (ReturnValue:Text)
   ) -> (then:Exec) {is=title K2Node_CallDelegate}
   Call On Level List Closed!(execute:Exec=@n3.CastFailed, self:Creative Property Editor Level List Object Reference) -> (then:Exec) {is=title K2Node_CallDelegate}
  ]}
 EventGraph {nodes [
   n1 = "On Item Selection Changed (ListView_Levels)"!() -> (OutputDelegate, then, Item:Object Reference, bIsSelected:bool) {is=title K2Node_ComponentBoundEvent}
   > Branch(Condition=@n1.bIsSelected)
   > UnbindCurrentLevelButton()
   > n2 = Cast To LevelOptionObjectWrapper(Object=@n1.Item)
   > n3 = Set MostRecentlySelectedLevelItem(MostRecentlySelectedLevelItem:Level Option Object Wrapper Object Reference=@n2) -> (Output_Get:Level Option Object Wrapper Object Reference)
   > UpdatedHeader!(execute, self, SelectedLevelOption:Level Option Object Wrapper Object Reference=@n3) -> (then)
   > UpdateCurrentLevelButton()
   n4 = "On Entry Generated (ListView_Levels)"!() -> (OutputDelegate, then, Widget:User Widget Object Reference) {is=title K2Node_ComponentBoundEvent}
   > n5 = Cast To CreativeLevelListItem!(execute, Object=@n4.Widget) -> (then, CastFailed, AsCreative Level List Item:Creative Level List Item Object Reference) {is=title K2Node_DynamicCast}
   > Branch(Condition=IsRepresentingLevelObject!(
     self:Creative Level List Item Object Reference=@n5
     InLevelObject:Level Option Object Wrapper Object Reference=Get MostRecentlySelectedLevelItem() -> (MostRecentlySelectedLevelItem:Level Option Object Wrapper Object Reference)
    ) -> (IsObject:bool))
   > n6 = Set CurrentLevelButton(CurrentLevelButton:Creative Level List Item Object Reference=@n5) -> (Output_Get:Creative Level List Item Object Reference)
   > Bind Event to On Level Committed(self:Creative Level List Item Object Reference=@n6, Delegate=Create Event()) {is=title K2Node_AddDelegate}
   n7 = "On Entry Released (ListView_Levels)"!() -> (OutputDelegate, then, Widget:User Widget Object Reference) {is=title K2Node_ComponentBoundEvent}
   > n8 = Cast To CreativeLevelListItem(Object=@n7.Widget)
   > Branch(Condition=IsRepresentingLevelObject(self=@n8, InLevelObject=Get MostRecentlySelectedLevelItem()))
   > UnbindCurrentLevelButton()
   n9 = "On Clicked (Button_CheckmarkBackground)"() {is=title K2Node_ComponentBoundEvent}
   CommitSelectedLevel!(execute=[@n9.then, @n10.then], self, bUseListName:bool=false) -> (then) {is=title K2Node_CallFunction}
   n10 = "On Button Clicked (UserActionMenuInputButton_Accept)"!() -> (OutputDelegate, then, Button:Common Button Legacy Object Reference) {is=title K2Node_ComponentBoundEvent}
   Event PreConstruct()
   > SetupCheckmarkBackgroundStyle()
   Event OnOutroAnimFinished() {is=title K2Node_Event}
   > UpdateCheckmarkBackgroundForControllerSelect!(execute, self, Is Selected:bool=false) -> (then)
   Event OnIntroAnimFinished() {is=title K2Node_Event}
   > n11 = Get!(execute:Exec) -> (then:Exec, else:Exec, CurrentLevelButton:Creative Level List Item Object Reference)
   SetSelectedIndex!(execute=@n11.else, self:List View Object Reference=Get ListView_Levels(), Index:int=0) -> (then)
   Event OnOpen() {is=title K2Node_Event}
   > UpdateCheckmarkBackgroundForControllerSelect(Is Selected=false)
   > SetReadOnly!(execute, self:WBP UIKit Filtered Input Field Object Reference=Get As Filtered Text Entry Widget(), IsReadOnly:bool=true) -> (then)
   Event OnClose() {is=title K2Node_Event}
   > Call On Level List Closed()
   Event Construct()
   > n12 = Cast To WBP_UIKit_FilteredInputField!(execute, Object=Get FilteredTextEntry_Widget()) -> (then, CastFailed, AsWBP UIKit Filtered Input Field:WBP UIKit Filtered Input Field Object Reference) {is=title K2Node_DynamicCast}
   > Set As Filtered Text Entry Widget(As Filtered Text Entry Widget:WBP UIKit Filtered Input Field Object Reference=@n12) -> (Output_Get:WBP UIKit Filtered Input Field Object Reference)
  ]}
 OnLevelListClosed {nodes [OnLevelListClosed() {is=title K2Node_FunctionEntry}]}
 OnLevelSelectionCommited {nodes [OnLevelSelectionCommited!() -> (then, Index:int, DisplayText:Text) {is=title K2Node_FunctionEntry}]}
 OnNewLevelButtonClicked {nodes [OnNewLevelButtonClicked() {is=title K2Node_FunctionEntry}]}
 SelectLevel {nodes [
   n1 = SelectLevel!() -> (then, LevelIndex:int) {is=title K2Node_FunctionEntry}
   > Branch(Condition="integer >= integer"(A=@n1))
   n2 = Get ListView_Levels()
   > ScrollIndexIntoView!(execute, self:List View Object Reference=@n2, Index:int=@n1) -> (then)
   > SetSelectedIndex(self=@n2, Index=@n1)
  ]}
 SetRenameEnabled {nodes [
   n1 = SetRenameEnabled!() -> (then, Enabled:bool=~) {is=title K2Node_FunctionEntry}
   > SetReadOnly(self=Get As Filtered Text Entry Widget(), IsReadOnly=NOT Boolean(A=@n1))
  ]}
 SetupCheckmarkBackgroundStyle {nodes [
   SetupCheckmarkBackgroundStyle() {is=title K2Node_FunctionEntry}
   > SetStyle!(
    execute
    self:Button Object Reference=Get Button_CheckmarkBackground() -> (Button_CheckmarkBackground:Button Object Reference)
    InStyle:"Button Style Structure (by ref)"=Make Button Style!(
     Normal:Slate Brush Structure=Get CheckBackground_Normal() -> (CheckBackground_Normal:Slate Brush Structure)
     Hovered:Slate Brush Structure=Get CheckBackground_Hovered() -> (CheckBackground_Hovered:Slate Brush Structure)
     Pressed:Slate Brush Structure=Get CheckBackground_Pressed() -> (CheckBackground_Pressed:Slate Brush Structure)
     Disabled:Slate Brush Structure=(TintColor=(SpecifiedColor=(R=1,G=1,B=1,A=1),ColorUseRule=UseColor_Specified),DrawAs=NoDrawType,Tiling=NoTile,Mirroring=NoMirror,ImageType=NoImage,ImageSize=(X=0,Y=0),Margin=(Left=0,Top=0,Right=0,Bottom=0),ResourceObject=None,OutlineSettings=(CornerRadii=(X=0,Y=0,Z=0,W=1),Color=(SpecifiedColor=(R=0,G=0,B=0,A=0),ColorUseRule=UseColor_Specified),Width=0,RoundingType=HalfHeightRadius,bUseBrushTransparency=False),UVRegion=(Min=(X=0,Y=0),Max=(X=0,Y=0),bIsValid=False),bIsDynamicallyLoaded=False,ResourceName="")
     NormalForeground:Slate Color Structure=(SpecifiedColor=(R=1,G=0,B=1,A=1),ColorUseRule=UseColor_Foreground)
     HoveredForeground:Slate Color Structure=(SpecifiedColor=(R=1,G=0,B=1,A=1),ColorUseRule=UseColor_Foreground)
     PressedForeground:Slate Color Structure=(SpecifiedColor=(R=1,G=0,B=1,A=1),ColorUseRule=UseColor_Foreground)
     DisabledForeground:Slate Color Structure=(SpecifiedColor=(R=1,G=0,B=1,A=1),ColorUseRule=UseColor_Foreground)
     NormalPadding:Margin Structure=(Left=0,Top=0,Right=0,Bottom=0)
     PressedPadding:Margin Structure=(Left=0,Top=0,Right=0,Bottom=0)
     PressedSlateSound:Slate Sound Structure=(ResourceObject=None)
     ClickedSlateSound:Slate Sound Structure=(ResourceObject=None)
     HoveredSlateSound:Slate Sound Structure=(ResourceObject=None)
    ) -> (ButtonStyle:Button Style Structure) {is=title K2Node_MakeStruct}
   ) -> (then)
  ]}
 UnbindCurrentLevelButton {nodes [
   UnbindCurrentLevelButton() {is=title K2Node_FunctionEntry}
   n1 = Get CurrentLevelButton()
   > Is Valid(InputObject=@n1)
   > Unbind all Events from On Level Committed!(execute:Exec, self:Creative Level List Item Object Reference=@n1) -> (then:Exec) {is=title K2Node_ClearDelegate}
   > Set CurrentLevelButton()
  ]}
 UpdateCheckmarkBackgroundForControllerSelect {nodes [
   n1 = UpdateCheckmarkBackgroundForControllerSelect!() -> (then, Is Selected:bool=~) {is=title K2Node_FunctionEntry}
   > Set members in Button Style!(
    execute:Exec
    StructRef:"Button Style Structure (by ref)"=Get WidgetStyle!(self:Button Object Reference=Get Button_CheckmarkBackground()) -> (WidgetStyle:Button Style Structure)
    Normal:Slate Brush Structure=Select(Option 0:Slate Brush Structure=Get CheckBackground_Normal(), Option 1:Slate Brush Structure=Get CheckBackground_Pressed(), Index=@n1) -> (ReturnValue:Slate Brush Structure)
   ) -> (then:Exec, StructOut:"Button Style Structure (by ref)") {is=title K2Node_SetFieldsInStruct}
  ]}
 UpdateCurrentLevelButton {nodes [
   UpdateCurrentLevelButton() {is=title K2Node_FunctionEntry}
   > Set CurrentLevelButton()
   > n1 = For Each Loop with Break!(
    Exec:Exec
    Array:Array of User Widget Object References=GetDisplayedEntryWidgets!(self:List View Base Object Reference=Get ListView_Levels()) -> (ReturnValue:Array of User Widget Object References)
    Break:Exec=@n4.then
   ) -> (LoopBody:Exec, Array Element:User Widget Object Reference, Array Index:int, Completed:Exec)
   > n2 = Cast To CreativeLevelListItem(Object=@n1)
   > Branch(Condition=IsRepresentingLevelObject(self=@n2, InLevelObject=Get MostRecentlySelectedLevelItem()))
   > n3 = Set CurrentLevelButton(CurrentLevelButton=@n2)
   > n4 = Bind Event to On Level Committed(self=@n3, Delegate=Create Event())
  ]}
 UpdatedHeader {nodes [
   n1 = UpdatedHeader!() -> (then, SelectedLevelOption:Level Option Object Wrapper Object Reference) {is=title K2Node_FunctionEntry}
   > SetInputFieldText!(
    execute
    self:WBP UIKit Filtered Input Field Object Reference=Get As Filtered Text Entry Widget()
    Text:Text=Get!(self:Level Option Object Wrapper Object Reference=@n1) -> (
     LevelOption_LevelInstanceGuid:Guid Structure
     LevelOption_LevelName:String
     LevelOption_bNewLevel:bool=~
     LevelOption_DisplayName:Text
     LevelOption_OptionValueName:String
     LevelOption_GrantedContextTags:Gameplay Tag Container Structure
    ).LevelOption_DisplayName
   ) -> (then)
  ]}
]
variables [
 As Filtered Text Entry Widget {type=WBP_UIKit_FilteredInputField_C, default=null}
 ButtonIcon_AddLevel {
  type=SlateBrush
  default {
   bIsDynamicallyLoaded=false
   drawAs=Image
   imageType=NoImage
   mirroring=NoMirror
   resourceName=None
   resourceObject=/Game/Athena/HUD/Spectator/Art/T-Icon-Broadcast-DBNO-Cross
   tiling=NoTile
   imageSize=<32 32>
   margin=<ltrb 0 0 0 0>
   outlineSettings {
    bUseBrushTransparency=false
    roundingType=HalfHeightRadius
    width=0
    color {colorUseRule=UseColor_Specified, specifiedColor=<rgba 0 0 0 0>}
    cornerRadii=<0 0 0 1>
   }
   tintColor {colorUseRule=UseColor_Specified, specifiedColor=<rgba 0.2159 0.6584 1 1>}
   uVRegion {bIsValid=false, max=<0 0>, min=<0 0>}
  }
 }
 ButtonIcon_ClearLevel {
  type=SlateBrush
  default {
   bIsDynamicallyLoaded=false
   drawAs=Image
   imageType=NoImage
   mirroring=NoMirror
   resourceName=None
   resourceObject=/Game/Athena/HUD/Loki/Icon_Undo
   tiling=NoTile
   imageSize=<32 32>
   margin=<ltrb 0 0 0 0>
   outlineSettings {
    bUseBrushTransparency=false
    roundingType=HalfHeightRadius
    width=0
    color {colorUseRule=UseColor_Specified, specifiedColor=<rgba 0 0 0 0>}
    cornerRadii=<0 0 0 1>
   }
   tintColor {colorUseRule=UseColor_Specified, specifiedColor=<rgba 0.2159 0.6584 1 1>}
   uVRegion {bIsValid=false, max=<0 0>, min=<0 0>}
  }
 }
 CheckBackground_Hovered {
  type=SlateBrush
  default {
   bIsDynamicallyLoaded=false
   drawAs=Image
   imageType=NoImage
   mirroring=NoMirror
   resourceName=None
   resourceObject=/Game/Athena/UI/InGame/Creative/PropertyEditor/Color_Picker/T_Icon_ColorPicker_action_bg
   tiling=NoTile
   imageSize=<94 81>
   margin=<ltrb 0 0 0 0>
   outlineSettings {
    bUseBrushTransparency=false
    roundingType=HalfHeightRadius
    width=0
    color {colorUseRule=UseColor_Specified, specifiedColor=<rgba 0 0 0 0>}
    cornerRadii=<0 0 0 1>
   }
   tintColor {colorUseRule=UseColor_Specified, specifiedColor=<rgba 0.18 0.737 0.878 1>}
   uVRegion {bIsValid=false, max=<0 0>, min=<0 0>}
  }
 }
 CheckBackground_Normal {
  type=SlateBrush
  default {
   bIsDynamicallyLoaded=false
   drawAs=Image
   imageType=NoImage
   mirroring=NoMirror
   resourceName=None
   resourceObject=&T_Icon_ColorPicker_action_bg
   tiling=NoTile
   imageSize=<94 81>
   margin=<ltrb 0 0 0 0>
   outlineSettings {
    bUseBrushTransparency=false
    roundingType=HalfHeightRadius
    width=0
    color {colorUseRule=UseColor_Specified, specifiedColor=<rgba 0 0 0 0>}
    cornerRadii=<0 0 0 1>
   }
   tintColor {colorUseRule=UseColor_Specified, specifiedColor=<rgba 0.094 0.631 0.847 1>}
   uVRegion {bIsValid=false, max=<0 0>, min=<0 0>}
  }
 }
 CheckBackground_Pressed {
  type=SlateBrush
  default {
   bIsDynamicallyLoaded=false
   drawAs=Image
   imageType=NoImage
   mirroring=NoMirror
   resourceName=None
   resourceObject=&T_Icon_ColorPicker_action_bg
   tiling=NoTile
   imageSize=<94 81>
   margin=<ltrb 0 0 0 0>
   outlineSettings {
    bUseBrushTransparency=false
    roundingType=HalfHeightRadius
    width=0
    color {colorUseRule=UseColor_Specified, specifiedColor=<rgba 0 0 0 0>}
    cornerRadii=<0 0 0 1>
   }
   tintColor {colorUseRule=UseColor_Specified, specifiedColor=<rgba 0.302 0.937 1 1>}
   uVRegion {bIsValid=false, max=<0 0>, min=<0 0>}
  }
 }
 CurrentButtonIcon {
  type=SlateBrush
  default {
   bIsDynamicallyLoaded=false
   drawAs=Image
   imageType=NoImage
   mirroring=NoMirror
   resourceName=None
   resourceObject=null
   tiling=NoTile
   imageSize=<32 32>
   margin=<ltrb 0 0 0 0>
   outlineSettings {
    bUseBrushTransparency=false
    roundingType=HalfHeightRadius
    width=0
    color {colorUseRule=UseColor_Specified, specifiedColor=<rgba 0 0 0 0>}
    cornerRadii=<0 0 0 1>
   }
   tintColor {colorUseRule=UseColor_Specified, specifiedColor=<rgba 1 1 1 1>}
   uVRegion {bIsValid=false, max=<0 0>, min=<0 0>}
  }
 }
 CurrentLevelButton {type=CreativeLevelListItem_C, default=null}
 CurrentLevelNameIsValid {type=boolean}
 MostRecentlySelectedLevelItem {type=LevelOptionObjectWrapper, default=null}
 OnLevelListClosed {}
 OnLevelSelectionCommited {}
 OnNewLevelButtonClicked {}
 PreviousSelectedLevelItem {type=Object, default=null}
]
widgets {
 info {
  inheritedWidgetCount=2
  parentClass=/Script/FortniteUI.FortCreativePropertyEditorLevelListDialog
  rootWidgetClass=&Overlay
  widgetCount=23
 }
 widgets [
  Overlay_18 {widgetClassPath=&Overlay, widget=Overlay_18}
  UserActionMenuInputButton_Back {
   widgetClassPath=/Game/UI/SocialPanel/UserActionMenuInputButton
   bIsVariable=true
   parent=Overlay_18
   widget=UserActionMenuInputButton_Back
   properties {
    bHideInputActionWithKeyboard=true
    bStopAction=true
    inputPriority=10
    onButtonBaseClicked=(&self:WidgetTree.UserActionMenuInputButton_Back.HandleOnButtonClicked)
    onButtonBaseDoubleClicked=(&self:WidgetTree.UserActionMenuInputButton_Back.HandleOnButtonDoubleClicked)
    onButtonBaseHovered=(&self:WidgetTree.UserActionMenuInputButton_Back.HandleOnButtonHovered)
    onButtonBaseUnhovered=(&self:WidgetTree.UserActionMenuInputButton_Back.HandleOnButtonUnhovered)
    onSelectedChangedBase=(&self:WidgetTree.UserActionMenuInputButton_Back.HandleOnSelectedChanged)
    priority=10
    visibility=Collapsed
    triggeringInputAction {dataTable=&InputActionDataTable, rowName=GenericBack}
   }
   slot {class=OverlaySlot, properties {padding=<ltrb 10 10 10 10>}}
  }
  ScaleBox_48 {
   widgetClassPath=&ScaleBox
   parent=Overlay_18
   widget=ScaleBox_48
   properties {stretch=Fill}
   slot {class=OverlaySlot, properties {verticalAlignment=VAlign_Fill, padding=<ltrb 0 0 0 -10>}}
  }
  SizeBox_WidgetSize {
   widgetClassPath=&SizeBox
   bIsVariable=true
   parent=ScaleBox_48
   widget=SizeBox_WidgetSize
   properties {bOverride_HeightOverride=true, bOverride_WidthOverride=true, heightOverride=350, widthOverride=438}
   slot {class=ScaleBoxSlot, properties {horizontalAlignment=HAlign_Left, verticalAlignment=VAlign_Top}}
  }
  Border_56 {
   widgetClassPath=&Border
   parent=SizeBox_WidgetSize
   widget=Border_56
   properties {
    horizontalAlignment=HAlign_Left
    verticalAlignment=VAlign_Top
    background {
     bIsDynamicallyLoaded=false
     drawAs=Image
     imageType=NoImage
     mirroring=NoMirror
     resourceName=None
     resourceObject=null
     tiling=NoTile
     imageSize=<32 32>
     margin=<ltrb 0 0 0 0>
     outlineSettings {
      bUseBrushTransparency=false
      roundingType=HalfHeightRadius
      width=0
      color {colorUseRule=UseColor_Specified, specifiedColor=<rgba 0 0 0 0>}
      cornerRadii=<0 0 0 0>
     }
     tintColor {colorUseRule=UseColor_Specified, specifiedColor=<rgba 0.0027 0.006 0.0452 1>}
     uVRegion {bIsValid=false, max=<0 0>, min=<0 0>}
    }
    desiredSizeScale=<0 0>
    padding=<ltrb 0 0 0 0>
   }
   slot {class=SizeBoxSlot}
  }
  Image_70 {
   widgetClassPath=&Image
   bIsVariable=true
   parent=Overlay_18
   widget=Image_70
   properties {
    visibility=SelfHitTestInvisible
    brush {
     bIsDynamicallyLoaded=false
     drawAs=Image
     imageType=NoImage
     mirroring=NoMirror
     resourceName=None
     resourceObject=/Game/Athena/UI/InGame/Creative/PropertyEditor/Color_Picker/T_Icon_ColorPicker_current_bg
     tiling=NoTile
     imageSize=<438 80>
     margin=<ltrb 0 0 0 0>
     outlineSettings {
      bUseBrushTransparency=false
      roundingType=HalfHeightRadius
      width=0
      color {colorUseRule=UseColor_Specified, specifiedColor=<rgba 0 0 0 0>}
      cornerRadii=<0 0 0 0>
     }
     tintColor {colorUseRule=UseColor_Specified, specifiedColor=<rgba 1 1 1 1>}
     uVRegion {bIsValid=false, max=<0 0>, min=<0 0>}
    }
   }
   slot {class=OverlaySlot}
  }
  SizeBox_37 {
   widgetClassPath=&SizeBox
   parent=Overlay_18
   widget=SizeBox_37
   properties {bOverride_HeightOverride=true, bOverride_WidthOverride=true, heightOverride=80, widthOverride=438}
   slot {class=OverlaySlot}
  }
  Overlay_0 {widgetClassPath=&Overlay, parent=SizeBox_37, widget=Overlay_0, slot {class=SizeBoxSlot}}
  Image_LeftHighlight {
   widgetClassPath=&Image
   parent=Overlay_0
   widget=Image_LeftHighlight
   properties {brush {
     bIsDynamicallyLoaded=false
     drawAs=Image
     imageType=NoImage
     mirroring=NoMirror
     resourceName=None
     resourceObject=/Game/Athena/UI/InGame/Creative/PropertyEditor/Color_Picker/T_Icon_ColorPicker_highlight
     tiling=NoTile
     imageSize=<10 80>
     margin=<ltrb 0 0 0 0>
     outlineSettings {
      bUseBrushTransparency=false
      roundingType=HalfHeightRadius
      width=0
      color {colorUseRule=UseColor_Specified, specifiedColor=<rgba 0 0 0 0>}
      cornerRadii=<0 0 0 0>
     }
     tintColor {colorUseRule=UseColor_Specified, specifiedColor=<rgba 1 1 1 1>}
     uVRegion {bIsValid=false, max=<0 0>, min=<0 0>}
    }}
   slot {class=OverlaySlot, properties {verticalAlignment=VAlign_Fill}}
  }
  SizeBox_0 {
   widgetClassPath=&SizeBox
   parent=Overlay_0
   widget=SizeBox_0
   properties {bOverride_HeightOverride=true, bOverride_WidthOverride=true, heightOverride=45, widthOverride=315}
   slot {class=OverlaySlot, properties {verticalAlignment=VAlign_Center, padding=<ltrb 20 0 0 0>}}
  }
  FilteredTextEntry_Widget {
   widgetClassPath=/FortUILibrary/UIKit/InputField/WBP_UIKit_FilteredInputField
   bInherited=true
   bIsVariable=true
   parent=SizeBox_0
   widget=FilteredTextEntry_Widget
   properties {hideValidationInfo=true, foregroundColor {colorUseRule=UseColor_Specified, specifiedColor=<rgba 0 0 0 1>}}
   slot {class=SizeBoxSlot}
  }
  InvalidNameErrorText {
   widgetClassPath=&CommonTextBlock
   bIsVariable=true
   parent=Overlay_0
   widget=InvalidNameErrorText
   properties {
    style=/FortUILibrary/Styles/Typography/Foundation/TextStyle-Base-XXS-B-Red.TextStyle-Base-XXS-B-Red_C
    text=Alphanumeric Text Only
    visibility=Hidden
    colorAndOpacity {colorUseRule=UseColor_Specified, specifiedColor=<rgba 1 0.1458 0.1458 1>}
    font {
     bForceMonospaced=false
     bMaterialIsStencil=false
     fontMaterial=null
     fontObject=&NotoSans
     letterSpacing=0
     monospacedWidth=1
     size=14
     skewAmount=0
     typefaceFontName=NotoSans-Bold
     outlineSettings {
      bApplyOutlineToDropShadows=false
      bMiteredCorners=false
      bSeparateFillAlpha=false
      outlineMaterial=null
      outlineSize=0
      outlineColor=<rgba 0 0 0 1>
     }
    }
    shadowOffset=<0 0>
   }
   slot {class=OverlaySlot, properties {verticalAlignment=VAlign_Bottom, padding=<ltrb 20 0 0 0>}}
  }
  Button_CheckmarkBackground {
   widgetClassPath=&Button
   parent=Overlay_0
   widget=Button_CheckmarkBackground
   properties {
    isFocusable=false
    widgetStyle {
     clickedSlateSound {resourceObject=null}
     disabled {
      bIsDynamicallyLoaded=false
      drawAs=NoDrawType
      imageType=NoImage
      mirroring=NoMirror
      resourceName=None
      resourceObject=null
      tiling=NoTile
      imageSize=<0 0>
      margin=<ltrb 0 0 0 0>
      outlineSettings {
       bUseBrushTransparency=false
       roundingType=HalfHeightRadius
       width=0
       color {colorUseRule=UseColor_Specified, specifiedColor=<rgba 0 0 0 0>}
       cornerRadii=<0 0 0 0>
      }
      tintColor {colorUseRule=UseColor_Specified, specifiedColor=<rgba 1 1 1 1>}
      uVRegion {bIsValid=false, max=<0 0>, min=<0 0>}
     }
     disabledForeground {colorUseRule=UseColor_Specified, specifiedColor=<rgba 0.5271 0.5271 0.5271 1>}
     hovered {
      bIsDynamicallyLoaded=false
      drawAs=RoundedBox
      imageType=NoImage
      mirroring=NoMirror
      resourceName=None
      resourceObject=&T_Icon_ColorPicker_action_bg
      tiling=NoTile
      imageSize=<94 81>
      margin=<ltrb 0 0 0 0>
      outlineSettings {
       bUseBrushTransparency=true
       roundingType=FixedRadius
       width=1
       color {colorUseRule=UseColor_Specified, specifiedColor=<rgba 0.7243 0.7243 0.7243 1>}
       cornerRadii=<4 4 4 4>
      }
      tintColor {colorUseRule=UseColor_Specified, specifiedColor=<rgba 1 1 1 0.5>}
      uVRegion {bIsValid=false, max=<0 0>, min=<0 0>}
     }
     hoveredForeground {colorUseRule=UseColor_Specified, specifiedColor=<rgba 0 1 1 1>}
     hoveredSlateSound {resourceObject=null}
     normal {
      bIsDynamicallyLoaded=false
      drawAs=RoundedBox
      imageType=NoImage
      mirroring=NoMirror
      resourceName=None
      resourceObject=&T_Icon_ColorPicker_action_bg
      tiling=NoTile
      imageSize=<94 81>
      margin=<ltrb 0 0 0 0>
      outlineSettings {
       bUseBrushTransparency=true
       roundingType=FixedRadius
       width=1
       color {colorUseRule=UseColor_Specified, specifiedColor=<rgba 0.6951 0.6951 0.6951 1>}
       cornerRadii=<4 4 4 4>
      }
      tintColor {colorUseRule=UseColor_Specified, specifiedColor=<rgba 1 1 1 0.25>}
      uVRegion {bIsValid=false, max=<0 0>, min=<0 0>}
     }
     normalForeground {colorUseRule=UseColor_Specified, specifiedColor=<rgba 0 1 1 1>}
     normalPadding=<ltrb 0 0 0 0>
     pressed {
      bIsDynamicallyLoaded=false
      drawAs=RoundedBox
      imageType=NoImage
      mirroring=NoMirror
      resourceName=None
      resourceObject=&T_Icon_ColorPicker_action_bg
      tiling=NoTile
      imageSize=<94 81>
      margin=<ltrb 0 0 0 0>
      outlineSettings {
       bUseBrushTransparency=true
       roundingType=FixedRadius
       width=1
       color {colorUseRule=UseColor_Specified, specifiedColor=<rgba 0.7243 0.7243 0.7243 1>}
       cornerRadii=<4 4 4 4>
      }
      tintColor {colorUseRule=UseColor_Specified, specifiedColor=<rgba 1 1 1 0.5>}
      uVRegion {bIsValid=false, max=<0 0>, min=<0 0>}
     }
     pressedForeground {colorUseRule=UseColor_Specified, specifiedColor=<rgba 0 1 1 1>}
     pressedPadding=<ltrb 0 0 0 0>
     pressedSlateSound {resourceObject=null}
    }
   }
   slot {class=OverlaySlot, properties {horizontalAlignment=HAlign_Right, verticalAlignment=VAlign_Fill}}
  }
  SizeBox_1 {
   widgetClassPath=&SizeBox
   parent=Overlay_0
   widget=SizeBox_1
   properties {bOverride_WidthOverride=true, visibility=HitTestInvisible, widthOverride=70}
   slot {class=OverlaySlot, properties {horizontalAlignment=HAlign_Right, verticalAlignment=VAlign_Center}}
  }
  HorizontalBox_312 {widgetClassPath=&HorizontalBox, parent=SizeBox_1, widget=HorizontalBox_312, slot {class=SizeBoxSlot}}
  UserActionMenuInputButton_Accept {
   widgetClassPath=&UserActionMenuInputButton
   bIsVariable=true
   parent=HorizontalBox_312
   widget=UserActionMenuInputButton_Accept
   properties {
    bHideInputActionWithKeyboard=true
    onButtonBaseClicked=(&self:WidgetTree.UserActionMenuInputButton_Accept.HandleOnButtonClicked)
    onButtonBaseDoubleClicked=(&self:WidgetTree.UserActionMenuInputButton_Accept.HandleOnButtonDoubleClicked)
    onButtonBaseHovered=(&self:WidgetTree.UserActionMenuInputButton_Accept.HandleOnButtonHovered)
    onButtonBaseUnhovered=(&self:WidgetTree.UserActionMenuInputButton_Accept.HandleOnButtonUnhovered)
    onSelectedChangedBase=(&self:WidgetTree.UserActionMenuInputButton_Accept.HandleOnSelectedChanged)
    triggeringInputAction {dataTable=&InputActionDataTable, rowName=VoteYes}
   }
   slot {class=HorizontalBoxSlot}
  }
  Overlay_1 {
   widgetClassPath=&Overlay
   parent=HorizontalBox_312
   widget=Overlay_1
   slot {class=HorizontalBoxSlot, properties {size {sizeRule=Fill, value=1}}}
  }
  Image_Checkmark {
   widgetClassPath=&Image
   parent=Overlay_1
   widget=Image_Checkmark
   properties {brush {
     bIsDynamicallyLoaded=false
     drawAs=Image
     imageType=NoImage
     mirroring=NoMirror
     resourceName=None
     resourceObject=/Game/UI/InGame/Mission/MissionActivation/Textures/T-Icon-VoteYes
     tiling=NoTile
     imageSize=<32 32>
     margin=<ltrb 0 0 0 0>
     outlineSettings {
      bUseBrushTransparency=false
      roundingType=HalfHeightRadius
      width=0
      color {colorUseRule=UseColor_Specified, specifiedColor=<rgba 0 0 0 0>}
      cornerRadii=<0 0 0 0>
     }
     tintColor {colorUseRule=UseColor_Specified, specifiedColor=<rgba 1 1 1 1>}
     uVRegion {bIsValid=false, max=<0 0>, min=<0 0>}
    }}
   slot {class=OverlaySlot, properties {padding=<ltrb 5 0 0 0>}}
  }
  VerticalBox_0 {
   widgetClassPath=&VerticalBox
   parent=Overlay_18
   widget=VerticalBox_0
   slot {class=OverlaySlot, properties {padding=<ltrb 20 90 0 0>}}
  }
  LevelListArea {
   widgetClassPath=&SizeBox
   parent=VerticalBox_0
   widget=LevelListArea
   properties {bOverride_HeightOverride=true, bOverride_WidthOverride=true, heightOverride=250, widthOverride=400}
   slot {class=VerticalBoxSlot}
  }
  ListView_Levels {
   widgetClassPath=&CommonListView
   bInherited=true
   bIsVariable=true
   parent=LevelListArea
   widget=ListView_Levels
   properties {
    bAllowDragging=false
    bReturnFocusToSelection=true
    entryWidgetClass=&/CreativeLevelListItem.CreativeLevelListItem_C
   }
   slot {class=SizeBoxSlot}
  }
  Overlay_Rename {
   widgetClassPath=&Overlay
   bIsVariable=true
   parent=VerticalBox_0
   widget=Overlay_Rename
   slot {class=VerticalBoxSlot, properties {horizontalAlignment=HAlign_Center, padding=<ltrb 0 5 0 5>}}
  }
  HorizontalBox_0 {widgetClassPath=&HorizontalBox, parent=Overlay_Rename, widget=HorizontalBox_0, slot {class=OverlaySlot}}
 ]
}
